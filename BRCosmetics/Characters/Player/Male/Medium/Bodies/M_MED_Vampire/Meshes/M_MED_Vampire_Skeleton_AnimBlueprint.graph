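# Post-pose Anim Blueprint for the male medium Vampire body, copying the owner mesh's pose and adding secondary motion to shoulders and jacket.
# Shoulder bones are constrained to the clavicles, then additively offset when the parachute opens or, otherwise, by alphas mapped from each upper arm's pitch relative to spine_05.
# The event graph computes those alphas, zeroing them under parachute.
# AnimDynamics then simulates back and front jacket bones with reduced gravity and high damping.
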AnimBlueprint /BRCosmetics/Characters/Player/Male/Medium/Bodies/M_MED_Vampire/Meshes/M_MED_Vampire_Skeleton_AnimBlueprint
target_skeleton=&/M_MED_Vampire_Skeleton
graphs [
 AnimGraph {nodes [
   n1 = Get MeshToCopy() -> (MeshToCopy:Skeletal Mesh Component Object Reference)
   n2 = "Constraint - dyn_shoulder_r (P)"(
    ComponentPose=Local To Component(LocalPose=Copy Pose From Mesh(SourceMeshComponent=@n1) {bUseAttachedParent=true})
    Alpha=1
    ConstraintWeights_0=1
   ) {
    alphaBoolBlend {blendOption=Linear}
    boneToModify {boneName=dyn_shoulder_r}
    constraintSetup [{targetBone {boneName=clavicle_r}}]
    constraintWeights [1]
   }
   n3 = "Constraint - dyn_shoulder_l (P)"(ComponentPose=@n2, Alpha=1, ConstraintWeights_0=1) {
    alphaBoolBlend {blendOption=Linear}
    boneToModify {boneName=dyn_shoulder_l}
    constraintSetup [{targetBone {boneName=clavicle_l}}]
    constraintWeights [1]
   }
   n4 = Get bIsParachuteOpened()
   n5 = "Transform (Modify) Bone - Bone: dyn_shoulder_r"!(ComponentPose=@n3, bAlphaBoolEnabled:bool=@n4, Translation="-2,0,2", Rotation="-15.000,0.000,0.000", Scale="1,1,1") -> (Pose) {
    alphaInputType=Bool
    rotationMode=BMM_Additive
    rotationSpace=BCS_BoneSpace
    translationMode=BMM_Additive
    translationSpace=BCS_BoneSpace
    alphaBoolBlend {blendInTime=1, blendOption=Linear, blendOutTime=1}
    boneToModify {boneName=dyn_shoulder_r}
    rotation {pitch=-15}
    translation {x=-2, z=2}
   }
   n6 = "Transform (Modify) Bone - Bone: dyn_shoulder_l"!(ComponentPose=@n5, bAlphaBoolEnabled:bool=@n4, Translation="-1,0,-1.25", Rotation="12,0.000,0.000", Scale="1,1,1") -> (Pose) {
    alphaInputType=Bool
    rotationMode=BMM_Additive
    rotationSpace=BCS_BoneSpace
    translationMode=BMM_Additive
    translationSpace=BCS_BoneSpace
    alphaBoolBlend {blendInTime=1, blendOption=Linear, blendOutTime=1}
    boneToModify {boneName=dyn_shoulder_l}
    rotation {pitch=12}
    translation {x=-1, z=-1.25}
   }
   n7 = "Transform (Modify) Bone - Bone: dyn_shoulder_r"!(
    ComponentPose=@n6
    Alpha:float=Get Clav R Rotation Alpha() -> (Clav R Rotation Alpha:double)
    Translation="-2,0,1.5"
    Rotation="-15,0,0.000"
    Scale="1,1,1"
   ) -> (Pose) {
    rotationMode=BMM_Additive
    rotationSpace=BCS_BoneSpace
    translationMode=BMM_Additive
    translationSpace=BCS_BoneSpace
    alphaBoolBlend {blendInTime=1, blendOption=Linear, blendOutTime=1}
    boneToModify {boneName=dyn_shoulder_r}
    rotation {pitch=-15}
    translation {x=-2, z=1.5}
   }
   n8 = "Transform (Modify) Bone - Bone: dyn_shoulder_l"!(
    ComponentPose=@n7
    Alpha:float=Get Clav L Rotation Alpha() -> (Clav L Rotation Alpha:double)
    Translation="-2,0,-1.5"
    Rotation="15,0.000,0.000"
    Scale="1,1,1"
   ) -> (Pose) {
    rotationMode=BMM_Additive
    rotationSpace=BCS_BoneSpace
    translationMode=BMM_Additive
    translationSpace=BCS_BoneSpace
    alphaBoolBlend {blendInTime=1, blendOption=Linear, blendOutTime=1}
    boneToModify {boneName=dyn_shoulder_l}
    rotation {pitch=15}
    translation {x=-2, z=-1.5}
   }
   n9 = Anim Dynamics - Bone: dyn_jacket_back(ComponentPose=@n8, Alpha=1, ExternalForce="0,0,0") {
    angularDampingOverride=0.99
    bGravityOverrideInSimSpace=true
    bOverrideAngularDamping=true
    gravityScale=0.35
    numSolverIterationsPostUpdate=2
    numSolverIterationsPreUpdate=8
    alphaBoolBlend {blendOption=Linear}
    boundBone {boneName=dyn_jacket_back}
    componentLinearAccScale=<0.5 0.5 0.5>
    componentLinearVelScale=<0.5 0.5 5>
    physicsBodyDefinitions [{
      sphereCollisionRadius=0
      boundBone {boneName=dyn_jacket_back}
      boxExtents=<30 30 15>
      constraintSetup {angularTargetAxis=AxisZ, twistAxis=AxisZ, angularLimitsMax=<10 5 6>, angularLimitsMin=<-10 -10 -6>, angularTarget=<0 0 0>}
      localJointOffset=<0 0 -3>
     }]
   }
   n10 = Anim Dynamics - Bone: dyn_jacket_front_l(ComponentPose=@n9, Alpha=1, ExternalForce="0,0,0") {
    angularDampingOverride=0.95
    bGravityOverrideInSimSpace=true
    bOverrideAngularDamping=true
    gravityScale=0.35
    numSolverIterationsPostUpdate=2
    numSolverIterationsPreUpdate=8
    alphaBoolBlend {blendOption=Linear}
    boundBone {boneName=dyn_jacket_front_l}
    componentLinearAccScale=<0.5 0.5 0.5>
    componentLinearVelScale=<0.5 0.5 5>
    physicsBodyDefinitions [{
      sphereCollisionRadius=0
      boundBone {boneName=dyn_jacket_front_l}
      boxExtents=<45 45 15>
      constraintSetup {angularTargetAxis=AxisZ, twistAxis=AxisZ, angularLimitsMax=<10 5 6>, angularLimitsMin=<-10 -10 -6>, angularTarget=<0 0 0>}
      localJointOffset=<0 0 -3>
     }]
   }
   n11 = Anim Dynamics - Bone: dyn_jacket_front_r(ComponentPose=@n10, Alpha=1, ExternalForce="0,0,0") {
    angularDampingOverride=0.95
    bGravityOverrideInSimSpace=true
    bOverrideAngularDamping=true
    gravityScale=0.35
    numSolverIterationsPostUpdate=2
    numSolverIterationsPreUpdate=8
    alphaBoolBlend {blendOption=Linear}
    boundBone {boneName=dyn_jacket_front_r}
    componentLinearAccScale=<0.5 0.5 0.5>
    componentLinearVelScale=<0.5 0.5 5>
    physicsBodyDefinitions [{
      sphereCollisionRadius=0
      boundBone {boneName=dyn_jacket_front_r}
      boxExtents=<45 45 15>
      constraintSetup {angularTargetAxis=AxisZ, twistAxis=AxisZ, angularLimitsMax=<10 5 6>, angularLimitsMin=<-10 -10 -6>, angularTarget=<0 0 0>}
      localJointOffset=<0 0 -3>
     }]
   }
   Output Pose(Result=Component To Local(ComponentPose=@n11)) {layerGroup=None}
  ]}
 EventGraph {nodes [
   Event BlueprintUpdateAnimation()
   n1 = Get bIsParachuteOpened()
   > n2 = Branch(Condition=@n1)
   > Set Clav R Rotation Alpha(Clav R Rotation Alpha:double=0) -> (Output_Get:double)
   n3 = BreakTransform(InTransform=GetDeltaTransformFromRefPose(self=GetOwnerSkeletalMeshComponent(), BoneName=upperarm_r, BaseName=spine_05))
   Set Clav R Rotation Alpha(execute=@n2.else, Clav R Rotation Alpha=MapRangeClamped(Value=BreakRotator(InRot=@n3.Rotation).Pitch, InRangeA=50, InRangeB=70))
   > n4 = Branch(Condition=@n1)
   > Set Clav L Rotation Alpha(Clav L Rotation Alpha:double=0) -> (Output_Get:double)
   n5 = BreakTransform(InTransform=GetDeltaTransformFromRefPose(self=GetOwnerSkeletalMeshComponent(), BoneName=upperarm_l, BaseName=spine_05))
   Set Clav L Rotation Alpha(execute=@n4.else, Clav L Rotation Alpha=MapRangeClamped(Value=BreakRotator(InRot=@n5.Rotation).Pitch, InRangeA=50, InRangeB=70))
   Event BlueprintInitializeAnimation()
   > Set MeshToCopy(MeshToCopy:Skeletal Mesh Component Object Reference=GetOwnerSkeletalMeshComponent()) -> (Output_Get:Skeletal Mesh Component Object Reference)
  ]}
]
variables [
 Clav L Rotation Alpha {type=number}
 Clav R Rotation Alpha {type=number}
 MeshToCopy {type=SkeletalMeshComponent}
]
